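# Physics asset for the one-handed female Flower Skeleton pickaxe, with bodies for its dangling rose and vine bones.
# A capsule covers the handle, small spheres sit on seven rose bones and two vine bones, and a small box forms the rose master bone.
# Each rose and vine constrains directly to that master in a star layout, with tight swing and twist limits on roses and looser ones on vines.
# The handle capsule has no constraint.
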PhysicsAsset /BRCosmetics/Weapons/FORT_Melee/Pickaxe_FlowerSkeletonFemale1h/Meshes/Flower_Skeleton_Female_1h_Axe_PhysicsAsset
bodies {
 Dyn_vine_a [{shapeType=Sphere, radius=2, center=<-1 0 0>}]
 Dyn_vine_b [{shapeType=Sphere, radius=2, center=<-1 0 0>}]
 Melee_TwoHanded_Handle [{shapeType=Capsule, length=38.8491, radius=3.1366, center=<-1.2879 0 41.118>, rotation=<rot 12.8099 0 0>}]
 dyn_rose_a [{shapeType=Sphere, radius=2, center=<-1 0 0>}]
 dyn_rose_b [{shapeType=Sphere, radius=2, center=<-1 0 0>}]
 dyn_rose_c [{shapeType=Sphere, radius=2, center=<-1 0 0>}]
 dyn_rose_d [{shapeType=Sphere, radius=2, center=<-1 0 0>}]
 dyn_rose_f [{shapeType=Sphere, radius=2, center=<-1 0 0>}]
 dyn_rose_g [{shapeType=Sphere, radius=2, center=<-1 0 0>}]
 dyn_rose_h [{shapeType=Sphere, radius=2, center=<-1 0 0>}]
 dyn_rose_mstr [{shapeType=Box, extentX=1.01, extentY=1.01, extentZ=1.01}]
}
constraints [
 {bone1Name=dyn_rose_a, bone2Name=dyn_rose_mstr, swing1LimitDegrees=15, swing2LimitDegrees=15, twistLimitDegrees=5}
 {bone1Name=dyn_rose_b, bone2Name=dyn_rose_mstr, swing1LimitDegrees=15, swing2LimitDegrees=15, twistLimitDegrees=5}
 {bone1Name=dyn_rose_c, bone2Name=dyn_rose_mstr, swing1LimitDegrees=15, swing2LimitDegrees=15, twistLimitDegrees=5}
 {bone1Name=dyn_rose_d, bone2Name=dyn_rose_mstr, swing1LimitDegrees=15, swing2LimitDegrees=15, twistLimitDegrees=5}
 {bone1Name=dyn_rose_f, bone2Name=dyn_rose_mstr, swing1LimitDegrees=15, swing2LimitDegrees=15, twistLimitDegrees=5}
 {bone1Name=dyn_rose_g, bone2Name=dyn_rose_mstr, swing1LimitDegrees=15, swing2LimitDegrees=15, twistLimitDegrees=5}
 {bone1Name=dyn_rose_h, bone2Name=dyn_rose_mstr, swing1LimitDegrees=15, swing2LimitDegrees=15, twistLimitDegrees=5}
 {bone1Name=Dyn_vine_b, bone2Name=dyn_rose_mstr, swing1LimitDegrees=45, swing2LimitDegrees=45, twistLimitDegrees=45}
 {bone1Name=Dyn_vine_a, bone2Name=dyn_rose_mstr, swing1LimitDegrees=45, swing2LimitDegrees=45, twistLimitDegrees=45}
]
